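# Niagara system for the bubble-pop effect of the HeadWhale cosmetic companion's emote.
# It has two emitters, Pop and Droplets, with no user parameters; their renderers and particle logic are not shown here.
# The system uses SystemState and a Set Variables assignment at system spawn, along with time-of-day scalar and color dynamic inputs, which suggests values that adapt to the time of day.
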NiagaraSystem /CosmeticCompanions/FX/Builtins/HeadWhale/Emote/NS_Emote_HeadWhale_BubblePop
dependencies {
 usedDataInterfaces []
 usedDynamicInputs [&FNTimeOfDayScalar, &FNTimeOfDayColor]
 usedModules [
  &SystemState
  &self:SystemSpawnScript.SystemScriptSource.SystemScriptGraph.NiagaraNodeAssignment_0.SetVariables_#1
 ]
}
emitters {Droplets {rendererClasses [], renderers []}, Pop {rendererClasses [], renderers []}}
summary {
 systemName=NS_Emote_HeadWhale_BubblePop
 emitters [{emitterName=Droplets, rendererClasses []}, {emitterName=Pop, rendererClasses []}]
 userVariables []
}
