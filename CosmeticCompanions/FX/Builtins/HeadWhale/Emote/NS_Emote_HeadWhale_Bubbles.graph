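# Niagara system for the HeadWhale cosmetic companion's bubble emote, with one active mesh emitter (Minimal) and a disabled Bubbles emitter.
# User.SpawnRate scales spawning of short-lived bubbles, offset by a sphere from skeletal mesh samples and inheriting velocity.
# Upward gravity, curl noise and drag move them, while point attraction pulls them toward their sampled position.
# Curves shrink and fade them, with brightness scaled by time of day.
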NiagaraSystem /CosmeticCompanions/FX/Builtins/HeadWhale/Emote/NS_Emote_HeadWhale_Bubbles
dependencies {
 usedDataInterfaces [&NiagaraDataInterfaceCurve, &NiagaraDataInterfaceSkeletalMesh]
 usedDynamicInputs []
 usedModules [
  &self:SystemSpawnScript.SystemScriptSource.SystemScriptGraph.NiagaraNodeAssignment_0.SetVariables_#1
  &SystemState
  &BlendSpaceHelper
 ]
}
emitters {
 Bubbles {bEnabled=false, rendererClasses [], renderers []}
 Minimal {
  emitterUpdateScript [
   &EmitterState()
   &SpawnRate(SpawnRate=&Multiply_Float(A=100, B=$User.SpawnRate), Use Spawn Probability=false)
  ]
  particleSpawnScript [
   &InitializeParticle(
    Color Maximum=<rgba 0.8802 0.9314 1 0.6>
    Color Minimum=<rgba 0.8958 0.9549 1 0.35>
    Color Mode=::NewEnumerator2
    Lifetime Max=1
    Lifetime Min=0.35
    Lifetime Mode=::NewEnumerator1
    Mesh Scale Mode=::NewEnumerator1
    Mesh Uniform Scale Max=0.125
    Mesh Uniform Scale Min=0.025
    UsePositionOffset=false
    Write Material Random=true
   )
   &ShapeLocation(MaskSpawnGroup=false, Offset=<0 0 -20>, Offset Mode=::NewEnumerator0, Override Seed=false, Sphere Radius=10, UseOwnerScale=true)
   &SkeletalMeshLocation(
    Bone Sampling Mode=::NewEnumerator4
    "Inherit Velocity (Scale)":NiagaraFloat=&FloatFromCurve(
     FloatCurve=DataInterface() {propertyValues {Curve {
        defaultValue=1
        keys [
         {arriveTangent=-0.4554, interpMode=RCIM_Cubic, leaveTangent=-0.4554, tangentMode=RCTM_User, time=0, value=1}
         {interpMode=RCIM_Cubic, tangentMode=RCTM_User, time=0.8, value=0}
        ]
       }}}
     Scale Curve=2.25
    )
    MaskSpawnGroup=false
    Position Sampling=::NewEnumerator1
    Sampled Position Offset:NiagaraPosition=$Particles.ShapeLocation.ShapePosition
    Skeletal Mesh=$System.Skeletal Mesh Reader
    Velocity Sampling=::NewEnumerator0
    VelocitySpeedLimit=false
   )
  ]
  particleUpdateScript [
   &ParticleState()
   &ScaleMeshSize(Scale Factor=&VectorFromFloat(Value=&FloatFromCurve(FloatCurve=DataInterface() {propertyValues {Curve {keys [<at 0 1>,<at 1 0>]}}})))
   &DynamicMaterialParameters(
    Index 3 Param 2=3
    Index 3 Param 4=&RandomRangeFloat(Minimum=0, Maximum=100)
    Module.Index3Float0Write=true
    Module.Index3Float1Write=true
    Module.Index3Float2Write=true
    Module.Index3Float3Write=true
    Write Parameter Index 0=::NewEnumerator0
    Write Parameter Index 3=::NewEnumerator1
   )
   &ScaleColor(
    Scale Alpha=&FloatFromCurve(FloatCurve=DataInterface() {propertyValues {Curve {keys [
         <at 0 0 RCIM_Cubic>
         {arriveTangent=1.4286, interpMode=RCIM_Cubic, leaveTangent=1.4286, time=0.35, value=1}
         {arriveTangent=0.5, interpMode=RCIM_Cubic, leaveTangent=0.5, time=0.7, value=1}
         {arriveTangent=-2, interpMode=RCIM_Cubic, leaveTangent=-2, time=0.85, value=1.25}
         <at 1 0.4 RCIM_Cubic>
        ]}}})
    Scale RGB=&VectorFromFloat(Value=&Multiply_Float(
      A=&FloatFromCurve(FloatCurve=DataInterface() {propertyValues {Curve {
          defaultValue=1
          keys [
           {interpMode=RCIM_Cubic, tangentMode=RCTM_User, time=0, value=0}
           {interpMode=RCIM_Cubic, tangentMode=RCTM_User, time=0.2, value=1}
           {tangentMode=RCTM_User, time=1, value=0}
          ]
         }}})
      B=&FNTimeOfDayScalar(Night Scale=0.325, Morning Scale=0.6, Day Scale=0.75, Evening Scale=0.6)
     ))
    ScaleA=true
    ScaleRGB=true
   )
   &CurlNoiseForce(
    Bias Noise=false
    Noise Frequency=80
    Noise Strength=&RandomRangeFloat(Minimum=325, Maximum=500)
    Pan Noise=false
   )
   &GravityForce(Gravity=&RandomRangeVector(Minimum=<0 0 325>, Maximum=<0 0 500>))
   &Drag(Drag=&RandomRangeFloat(Minimum=12, Maximum=15), Use Linear Drag=true, Use Rotational Drag=true)
   &PointAttractionForce(
    Attraction Radius:NiagaraFloat=&FloatFromCurve(FloatCurve=DataInterface() {propertyValues {Curve {keys [<at 0 1>,<at 1 0.5>]}}}, Scale Curve=30)
    Attraction Strength=&Add_Float(
     A=&Add_Float(
      A=&RandomRangeFloat(Minimum=26, Maximum=18)
      B=&FloatFromCurve(
       FloatCurve=DataInterface() {propertyValues {Curve {keys [<at 0 1>,<at 1 0>]}}}
       Scale Curve=&RandomRangeFloat(Minimum=7, Maximum=3)
      )
     )
     B=&FloatFromCurve(FloatCurve=DataInterface() {propertyValues {Curve {keys [<at 0.2 0>,<at 0.4 0>]}}}, Scale Curve=50)
    )
    Attractor Position:NiagaraPosition=$Particles.SkeletalMeshLocation.SampledPosition
    Attractor Position Offset:Vector3f=&RandomRangeVector(Minimum=<5 5 5>, Maximum=<-5 -5 -5>)
    Kill Within Radius=false
    Module.Use Falloff=true
    UseAttractorOffset=true
   )
   &SolveForcesAndVelocity(Clamp Velocity=false, Limit Acceleration=false)
  ]
  rendererClasses [&NiagaraMeshRendererProperties]
  renderers [{rendererClass=&NiagaraMeshRendererProperties, rendererIndex=0}]
 }
}
summary {
 systemName=NS_Emote_HeadWhale_Bubbles
 emitters [
  {bEnabled=false, emitterName=Bubbles, rendererClasses []}
  {emitterName=Minimal, rendererClasses [&NiagaraMeshRendererProperties]}
 ]
 userVariables [User.SpawnRate {type=NiagaraFloat, defaultValue=1, description=""}]
}
