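# Niagara system named for the burp emote effect of the HeadWhale cosmetic companion.
# It contains a single emitter, Bubbles, and uses the SystemState module.
# The file lists no renderers, user variables, data interfaces or dynamic inputs, so it does not show how the bubbles look or move.
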NiagaraSystem /CosmeticCompanions/FX/Builtins/HeadWhale/Emote/NS_Emote_HeadWhale_Burp
dependencies {usedDataInterfaces [], usedDynamicInputs [], usedModules [&SystemState]}
emitters {Bubbles {rendererClasses [], renderers []}}
summary {systemName=NS_Emote_HeadWhale_Burp, emitters [{emitterName=Bubbles, rendererClasses []}], userVariables []}
